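# Sound cue for the thrust-start sound of the Storm Soldier glider.
# A random node picks one of three equally weighted thruster waves and feeds it through a modulator that slightly varies volume.
# The cue plays at reduced volume through shared glider-thrust attenuation and concurrency settings, under the glider landing sound class.
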SoundCue /BRCosmetics/Sounds/Gliders/StormSoldier/Glider_StormSoldier_Thrust_Start_Cue
root=Modulator_0
nodes [
 {id=Modulator_0, children [Random_1], properties {volumeMax=1, volumeMin=0.85}}
 {id=Random_1, children [WavePlayer_1,WavePlayer_3,WavePlayer_4], properties {weights [1,1,1]}}
 {id=WavePlayer_1, properties {soundWaveAssetPtr=&/Glider_StormSoldier_Thruster_01}}
 {id=WavePlayer_3, properties {soundWaveAssetPtr=&/Glider_StormSoldier_Thruster_02}}
 {id=WavePlayer_4, properties {soundWaveAssetPtr=&/Glider_StormSoldier_Thruster_03}}
]
properties {
 attenuationSettings=&Fort_Player_Glider_Thrust_Athena_Att
 maxDistance=12500
 soundClassObject=&Fort_Gliders_Land
 volumeMultiplier=0.2
 concurrencySet [&Athena_Glider_Thrust_Conc]
}
